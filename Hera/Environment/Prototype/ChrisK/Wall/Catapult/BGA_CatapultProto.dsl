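; Prototype catapult building actor, set invulnerable, that fires once on the authority after a random delay following BeginPlay.
; BeginPlay caches start and end transforms, then FireCatapult's timeline lerps the arm, counterweight and mounted boulders between them.
; Timeline events hide the mounted boulders and call YeetStone, which on authority spawns and wakes a physics boulder at the arrow and applies a forward impulse.
; A later Reset event shows the mounted boulders again.
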(event EventBeginPlay
  (bind _catapultarm_a (Variables|Default|GetCatapultArm_A))
  (bind _relativerotation (Variables|/Script/Engine.SceneComponent|GetRelativeRotation _catapultarm_a))
  (Variables|BuildingActor|SetbIsInvulnerable true)
  (Variables|Default|SetArmStartLocation _relativerotation)
  (Variables|Default|SetArmEndRotation (Math|Rotator|CombineRotators _relativerotation "135.000000,0.000000,0.000000"))
  (bind _output_get (Variables|Default|SetWeightStartLoc (Variables|/Script/Engine.SceneComponent|GetRelativeLocation (Variables|Default|GetCatapultWeight_A))))
  (Variables|Default|SetWeightEndLoc (- _output_get "-50.000000,0.000000,85.000000"))
  (Variables|Default|SetBoulderRelativeEndLocation (Variables|/Script/Engine.SceneComponent|GetRelativeLocation (Variables|Default|GetArrow)))
  (Variables|Default|SetBoulderRelativeStartLocation (Variables|/Script/Engine.SceneComponent|GetRelativeLocation (Variables|Default|GetPhysics_Boulders)))
  (Networking|SwitchHasAuthority
    (:Authority
      (Utilities|FlowControl|Delay (Math|Random|RandomFloatinRange 5.0 7.0))
      (CallFunction|FireCatapult)
      (Utilities|FlowControl|Delay (Math|Random|RandomFloatinRange 5.0 7.0)))))

(event Custom|FireCatapult
  (Utilities|FlowControl|Sequence
    (:then_0
      (bind (_direction _lerp) (Timeline|Fire PlayFromStart)
        (:Update
          (Transformation|SetRelativeRotation (Variables|Default|GetCatapultArm_A) (Math|Rotator|Lerp(Rotator) (Variables|Default|GetArmStartLocation) (Variables|Default|GetArmEndRotation) _lerp true) false false)
          (Transformation|SetRelativeLocationAndRotation (Variables|Default|GetCatapultWeight_A) (Math|Vector|Lerp(Vector) (Variables|Default|GetWeightStartLoc) (Variables|Default|GetWeightEndLoc) _lerp) "0, 0, 0" false false)
          (Transformation|SetRelativeLocation (Variables|Default|GetPhysics_Boulders) (Math|Vector|Lerp(Vector) (Variables|Default|GetBoulderRelativeStartLocation) (Math|Vector|MakeVector (+ (.x (Variables|Default|GetBoulderRelativeEndLocation)) 1e+02) (.y (Variables|Default|GetBoulderRelativeEndLocation)) (.z (Variables|Default|GetBoulderRelativeEndLocation))) (Math|Float|Clamp(Float) (- _lerp 0.2))) false false))
        (:Yeetnow
          (Rendering|SetVisibility (Variables|Default|GetPhysics_Boulders) false false)
          (CallFunction|YeetStone))
        (:Reset
          (Rendering|SetVisibility (Variables|Default|GetPhysics_Boulders) true false))))))

(event Custom|YeetStone
  (Networking|SwitchHasAuthority
    (:Authority
      (Development|PrintString "Spawning")
      (bind _returnvalue (Game|SpawnActorfromClass "/Hera/Environment/Prototype/ChrisK/Wall/Catapult/Prop_PhysicsBoulder_Catapult.Prop_PhysicsBoulder_Catapult_C" (Math|Conversions|ToTransform(Vector) (Transformation|GetWorldLocation (Variables|Default|GetArrow))) "AlwaysSpawn" "MultiplyWithRoot" 0 0 0 0))
      (PhysicsObject|WakeUpPhysicsObject _returnvalue "AlwaysInstant")
      (Physics|AddImpulse (Variables|/Script/FortniteGame.BuildingSMActor|GetStaticMeshComponent _returnvalue) (* (Transformation|GetForwardVector (Variables|Default|GetArrow)) (Math|Conversions|ToVector(Float) 5e+03)) "None" true))))

(fn ConstructionScript)
