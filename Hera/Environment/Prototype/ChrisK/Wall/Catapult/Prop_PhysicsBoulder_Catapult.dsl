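; Catapult physics boulder prop that adds forward roll on impact and plays Niagara effects for spawning, rolling and destruction.
; Each hit applies a horizontal velocity-change impulse along its current travel, scaled by a multiplier that steps down with every hit.
; Non-server instances spawn rolling effects at hit points while the boulder moves fast enough.
; Optional spawn effects play at BeginPlay, and death spawns a destroyed effect; rolling and destroyed effects receive the mesh velocity.
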(event EventOnDeathPlayEffects
  (params (Damage :type Float%20%28single-precision%29) (DamageTags :type Gameplay%20Tag%20Container%20Structure%20%28by%20ref%29) (Momentum :type Vector :default "0, 0, 0") (HitInfo :type Hit%20Result%20Structure%20%28by%20ref%29) (InstigatedBy :type Fort%20Pawn%20Object%20Reference) (DamageCauser :type Actor%20Object%20Reference) (EffectContext :type Gameplay%20Effect%20Context%20Handle%20Structure))
  (bind _staticmeshcomponent (Variables|Debug|GetStaticMeshComponent))
  (|CallParentFunction Damage DamageTags Momentum HitInfo InstigatedBy DamageCauser EffectContext)
  (bind _returnvalue (Niagara|SpawnSystematLocation "/Game/Environments/Apollo/Rocks/PhysicsBoulder/NS_Physics_Boulder_Destroyed.NS_Physics_Boulder_Destroyed" (Transformation|GetWorldLocation _staticmeshcomponent)))
  (Niagara|SetNiagaraVariable(Vector3) _returnvalue "User.Momentum" (Transformation|GetComponentVelocity _staticmeshcomponent)))

(event Collision|EventHit
  (params (MyComp :type Primitive%20Component%20Object%20Reference) (Other :type Actor%20Object%20Reference) (OtherComp :type Primitive%20Component%20Object%20Reference) (bSelfMoved :type Boolean) (HitLocation :type Vector :default "0, 0, 0") (HitNormal :type Vector :default "0, 0, 0") (NormalImpulse :type Vector :default "0, 0, 0") (Hit :type Hit%20Result%20Structure%20%28by%20ref%29))
  (Physics|AddImpulse (Variables|Debug|GetStaticMeshComponent) (* (Math|Vector|MakeVector (.x (Transformation|GetVelocity self)) (.y (Transformation|GetVelocity self))) (Math|Conversions|ToVector(Float) (Variables|Default|GetForwardRollVelolcityMultiplier))) "None" true)
  (Variables|Default|SetForwardRollVelolcityMultiplier (Math|Float|Clamp(Float) (- (Variables|Default|GetForwardRollVelolcityMultiplier) 0.15)))
  (|CallParentFunction MyComp Other OtherComp bSelfMoved HitLocation HitNormal NormalImpulse Hit)
  (if (< (Math|Vector|VectorLength (Transformation|GetComponentVelocity (Variables|Debug|GetStaticMeshComponent))) 3e+02)
    _
    (elif (Networking|IsServer)
      _
      (else
        (bind _ns_rollingeffects (Variables|Effects|GetNS_RollingEffects)
          (:then
            (bind _returnvalue (Niagara|SpawnSystematLocation _ns_rollingeffects HitLocation))
            (Niagara|SetNiagaraVariable(Vector3) _returnvalue (Variables|Effects|GetRollingEffectsVector3VarName) (Transformation|GetComponentVelocity (Variables|Debug|GetStaticMeshComponent))))
          (:else))))))

(event EventBeginPlay
  (|CallParentFunction)
  (bind _onspawnfx_one (Variables|Effects|GetOnSpawnFX_One)
    (:then
      (Niagara|SpawnSystematLocation _onspawnfx_one (+ (Transformation|GetWorldLocation (Variables|Debug|GetStaticMeshComponent)) "0,0,-150.000000")))
    (:else
      _))
  (bind _onspawnfx_two (Variables|Effects|GetOnSpawnFX_Two)
    (:then
      (Niagara|SpawnSystemAttached _onspawnfx_two (Variables|Debug|GetStaticMeshComponent)))
    (:else)))

(fn ConstructionScript
  (|CallParentFunction))
